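# Translucent unlit, two-sided Niagara material for the smoke trail of the progressive Cowgirl pickaxe, tinted by blending particle color with LowColor.
# Opacity sums panning caustics and inky smoke, eroded by a dynamic parameter, with a noise-distorted trail texture that SmokeFadeto fades.
# UVs can derive from a world-space sphere mask around the offset object position.
# Edge masks and inverse Fresnel soften it; mobile skips distortion and depth fade.
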Material /BRCosmetics/Weapons/FORT_Melee/Pickaxe_Cowgirl_Progressive/FX/UpdatedFX/M_Cowgirl_Pickaxe_SmokeTrail
graph {
 properties {
  bAutomaticallySetUsageInEditor=false
  bUsedWithNiagaraMeshParticles=true
  bUsedWithNiagaraRibbons=true
  bUsedWithNiagaraSprites=true
  bUsedWithStaticMesh=false
  blendMode=BLEND_Translucent
  pixelDepthOffsetMode=PDOM_Legacy
  shadingModel=MSM_Unlit
  twoSided=true
  usedShadingModels=Unlit
 }
 nodes [
  n1 = VectorParameter() {parameterName=LowColor, defaultValue=<rgba 0.1339 0.5363 0.558 1>}
  n2 = ParticleColor()
  n3 = Desaturation(None=@n2.RGB, Fraction=Constant() {r=0.5})
  MaterialFunctionCall(Emissive=LinearInterpolate(A=Multiply(A=@n1.RGB, B=@n3), B=@n2.RGB)) {materialFunction=&MF_FX_EmissiveAdjustments}
  n4 = SphereMask(
   A=WorldPosition().XYZ
   B=Add(A=ObjectPositionWS(), B=VectorParameter() {parameterName=Initial Offset, defaultValue=<rgba 0 0 -875 0>}.RGB)
   Radius=Constant() {r=2550}
   Hardness=Constant()
  )
  n5 = Add(A=Multiply(A=@n4, B=@n4), B=Constant() {r=-200}) {constB=-20}
  n6 = TextureCoordinate()
  n7 = ScalarParameter() {defaultValue=0.25, parameterName=TileX2}
  n8 = ScalarParameter() {defaultValue=3, parameterName=TileY2}
  n9 = StaticSwitchParameter(
   True=AppendVector(A=Multiply(A=@n5) {constB=7}, B=ComponentMask(None=@n6) {g=true})
   False=Multiply(A=TextureCoordinate(), B=AppendVector(A=@n7, B=@n8))
  ) {defaultValue=true, parameterName=WorldPositionUVs}
  n10 = TextureCoordinate()
  n11 = Multiply(A=ComponentMask(None=@n10) {r=true}, B=ScalarParameter() {defaultValue=1, parameterName=U Tiling})
  n12 = DynamicParameter() {defaultValue=<rgba 1 0.5 0 0>, paramNames [SmokeFadeto, U Tiling, V Tiling, Erode]}
  n13 = Multiply(A=ComponentMask(None=@n10) {g=true}, B=ScalarParameter() {defaultValue=1, parameterName=V Tiling})
  n14 = LinearInterpolate(
   A=@n9
   B=AppendVector(A=Multiply(A=@n11, B=@n12.U Tiling), B=Multiply(A=@n13, B=@n12.V Tiling))
   Alpha=ScalarParameter() {parameterName=UseRandomUVTiling}
  )
  n15 = AppendVector(A=ScalarParameter() {defaultValue=-2, parameterName=SpeedX1}, B=ScalarParameter() {parameterName=SpeedY1})
  n16 = LinearInterpolate(Alpha=TextureSample(UVs=Panner(Coordinate=@n14, Speed=@n15) {speedX=2}) {samplerSource=SSM_Wrap_WorldGroupSettings, texture=&T_ShieldCaustics}.R) {constA=0.2, constB=3}
  n17 = MaterialFunctionCall(Float2=@n9) {materialFunction=&BreakOutFloat2Components}
  n18 = AppendVector(A=@n17.R, B=Add(A=@n17.G) {constB=0.5})
  n19 = AppendVector(A=ScalarParameter() {defaultValue=-2.4, parameterName=SpeedX2}, B=ScalarParameter() {parameterName=SpeedY2})
  n20 = LinearInterpolate(Alpha=TextureSample(UVs=Panner(Coordinate=@n18, Speed=@n19) {speedX=-2.4}) {samplerSource=SSM_Wrap_WorldGroupSettings, texture=&T_Inky_Smoke_Tile}.R) {constB=3}
  n21 = TextureCoordinate()
  n22 = Constant() {r=0.5}
  n23 = SphereMask(
   A=Power(Base=ComponentMask(None=@n21) {r=true}, Exp=ScalarParameter() {defaultValue=0.5, parameterName=OpatcityMaskUbias}) {constExponent=0.5}
   B=@n22
  ) {attenuationRadius=0.5, hardnessPercent=0}
  n24 = TextureCoordinate()
  n25 = Saturate(None=SphereMask(A=ComponentMask(None=@n24) {g=true}, B=@n22) {attenuationRadius=0.5, hardnessPercent=0})
  n26 = Multiply(A=Multiply(A=@n23, B=@n23), B=Multiply(A=@n25, B=@n25))
  n27 = Subtract(
   A=Multiply(A=Add(A=@n16, B=@n20), B=@n26)
   B=DynamicParameter() {defaultValue=<rgba 1 0.5 0 0>, paramNames [SmokeFadeto, U Tiling, V Tiling, Erode]}.Erode
  )
  n28 = SphereMask(
   A=WorldPosition().XYZ
   B=Add(A=ObjectPositionWS(), B=VectorParameter() {parameterName=Initial Offset_1, defaultValue=<rgba 0 0 -875 0>}.RGB)
   Radius=Constant() {r=2550}
   Hardness=Constant()
  )
  n29 = Multiply(A=@n28, B=@n28)
  n30 = Constant() {r=-200}
  n31 = AppendVector(A=Multiply(A=Add(A=@n29, B=@n30) {constB=-20}) {constB=22}, B=ComponentMask(None=TextureCoordinate()) {g=true})
  n32 = TextureSample(UVs=Panner(Coordinate=@n31) {speedX=-2.4}) {
   samplerSource=SSM_Wrap_WorldGroupSettings
   samplerType=SAMPLERTYPE_LinearColor
   texture=&T_NoisePack_LinearColor
  }
  n33 = ComponentMask(None=@n32.RGB) {g=true, r=true}
  n34 = ScalarParameter() {defaultValue=0.2, parameterName=Smoke Trail Distortion Intensity}
  n35 = Add(A=@n31, B=ShadingPathSwitch(Default=Multiply(A=@n33, B=@n34) {constB=0.45}, Mobile=Constant()) {desc=Migrated from FeatureLevelSwitch})
  n36 = AppendVector(A=Constant(), B=ScalarParameter() {defaultValue=-0.125, parameterName=Texture Offset})
  n37 = TextureSampleParameter2D(UVs=Panner(Coordinate=Add(A=@n35, B=@n36)) {speedX=-1.9}) {parameterName=SmokeTrailTexture, samplerType=SAMPLERTYPE_Grayscale, texture=&T_Smoke_Trail_Soft}
  n38 = Multiply(A=Multiply(A=@n37.R, B=@n37.R), B=@n26)
  n39 = TextureSample(UVs=Panner(Coordinate=Add(A=@n31) {constB=0.3}) {speedX=2.4, speedY=0.25}) {
   samplerSource=SSM_Wrap_WorldGroupSettings
   samplerType=SAMPLERTYPE_LinearColor
   texture=&T_NoisePack_LinearColor
  }
  n40 = TextureSample(UVs=Panner(Coordinate=@n31) {speedX=1.5, speedY=-0.18}) {
   samplerSource=SSM_Wrap_WorldGroupSettings
   samplerType=SAMPLERTYPE_LinearColor
   texture=&T_NoisePack_LinearColor
  }
  n41 = Multiply(A=Multiply(A=@n39.R, B=@n40.G), B=ComponentMask(None=TextureCoordinate()) {r=true})
  n42 = Multiply(A=@n41, B=@n41)
  n43 = Clamp(None=Multiply(A=Subtract(A=@n38, B=@n42), B=ScalarParameter() {defaultValue=20, parameterName=Flame Intensity}) {constB=40})
  n44 = ScalarParameter() {defaultValue=0.5, parameterName=Flame_Alpha}
  n45 = Add(
   A=Clamp(None=@n27, Max=@n27)
   B=LinearInterpolate(
    A=Multiply(A=@n43, B=@n44)
    Alpha=DynamicParameter() {defaultValue=<rgba 1 0.5 0 0>, paramNames [SmokeFadeto, U Tiling, V Tiling, Erode]}.SmokeFadeto
   ) {constB=0}
  )
  n46 = Multiply(A=Multiply(A=@n2.A, B=Clamp(None=@n45)), B=OneMinus(None=Fresnel() {baseReflectFraction=0, exponent=1}))
  ShadingPathSwitch(Default=DepthFade(Opacity=@n46, FadeDistance=ScalarParameter() {defaultValue=100, parameterName=DepthFadeDistance}), Mobile=@n46) {desc=Migrated from FeatureLevelSwitch}
 ]
}
