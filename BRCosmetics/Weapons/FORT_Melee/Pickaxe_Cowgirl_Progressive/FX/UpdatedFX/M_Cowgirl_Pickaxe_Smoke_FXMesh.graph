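# Translucent, unlit, two-sided smoke material for an FX mesh on the progressive Cowgirl pickaxe.
# Vertices displace along normals with panning noise, plus drag and lift offsets weighted by UV gradients.
# Opacity combines flowmap-warped inky smoke and caustics with UV edge, Fresnel and depth fades, then vertex-color noise erosion or switchable near-camera fade, scaled by SwingFade.
# Emissive blends two tints with a slow sine pulse.
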Material /BRCosmetics/Weapons/FORT_Melee/Pickaxe_Cowgirl_Progressive/FX/UpdatedFX/M_Cowgirl_Pickaxe_Smoke_FXMesh
graph {
 properties {
  bAllowNegativeEmissiveColor=true
  bTangentSpaceNormal=false
  blendMode=BLEND_Translucent
  pixelDepthOffsetMode=PDOM_Legacy
  shadingModel=MSM_Unlit
  translucencyLightingMode=TLM_VolumetricPerVertexDirectional
  twoSided=true
  usedShadingModels=Unlit
 }
 nodes [
  n1 = VertexNormalWS()
  n2 = TextureCoordinate() {uTiling=2}
  n3 = ScalarParameter() {defaultValue=1, parameterName=NoiseTile}
  n4 = Time()
  n5 = TextureSample(UVs=Panner(Coordinate=Multiply(A=@n2, B=@n3), Time=Multiply(A=@n4) {constB=4}) {speedX=0.05, speedY=0.1}) {samplerType=SAMPLERTYPE_Grayscale, texture=&LowResBlurredNoise}
  n6 = Multiply(A=ScalarParameter() {defaultValue=10, parameterName=Amount_Skirt}, B=@n5.RGB) {constB=350}
  n7 = VectorParameter() {parameterName=SkirtDrag}
  n8 = Constant()
  n9 = Add(
   A=Add(A=Multiply(A=@n1, B=@n6), B=StaticSwitchParameter(True=@n7.RGB, False=@n8) {defaultValue=true, parameterName=Use Skirt Drag})
   B=Transform(None=VectorParameter() {parameterName=LocalPostioning}.RGB) {transformSourceType=TRANSFORMSOURCE_Local}
  )
  n10 = OneMinus(None=ComponentMask(None=TextureCoordinate()) {g=true})
  n11 = Multiply(A=@n10, B=@n10)
  n12 = OneMinus(None=ComponentMask(None=TextureCoordinate()) {g=true})
  n13 = Power(Base=@n12, Exp=ScalarParameter() {defaultValue=1.25, parameterName=Z_Grad Contrast})
  n14 = Multiply(A=Constant3Vector() {constant=<rgba 0 0 1 0>}, B=ScalarParameter() {defaultValue=5, parameterName=ZAmount})
  n15 = Multiply(A=VertexNormalWS()) {constB=2.5}
  n16 = LinearInterpolate(Alpha=NamedRerouteUsage()) {constA=-1}
  Add(A=Add(A=Multiply(A=@n9, B=@n11), B=Multiply(A=@n13, B=@n14)), B=Multiply(A=@n12, B=Multiply(A=@n15, B=@n16)))
  n17 = Constant() {r=0.5}
  n18 = VectorParameter() {parameterName=BaseColor, defaultValue=<rgba 1 1 1 1>}
  n19 = LinearInterpolate(
   A=Multiply(A=VectorParameter() {parameterName=LowColor, defaultValue=<rgba 0.1339 0.5363 0.558 1>}.RGB, B=Desaturation(None=@n18.RGB, Fraction=@n17))
   B=@n18.RGB
  )
  n20 = Sine(None=Multiply(A=Constant() {r=0.22}, B=Time()))
  n21 = Multiply(A=Clamp(None=Abs(None=@n20)) {minDefault=0.6}) {constB=0.5}
  MaterialFunctionCall(Emissive=Multiply(A=Multiply(A=@n19, B=@n21), B=ScalarParameter() {defaultValue=7, parameterName=Emissive Intensity})) {materialFunction=&MF_FX_EmissiveAdjustments}
  n22 = TextureCoordinate() {uTiling=2, vTiling=0.15}
  n23 = ScalarParameter() {defaultValue=1, parameterName=NoiseTile}
  n24 = TextureSample(UVs=Panner(Coordinate=Multiply(A=@n22, B=@n23)) {speedX=0.05}) {samplerSource=SSM_Wrap_WorldGroupSettings, texture=&T_ShieldCaustics}
  n25 = ComponentMask(None=TextureCoordinate()) {g=true}
  n26 = Multiply(A=Multiply(A=ConstantBiasScale(None=@n24.R), B=Power(Base=@n25) {constExponent=5}))
  n27 = Multiply(A=@n23, B=TextureCoordinate() {vTiling=0.1})
  n28 = TextureCoordinate()
  n29 = Add(A=@n28, B=Constant2Vector() {g=0.3, r=0.1})
  n30 = VertexColor()
  n31 = TextureSample(UVs=Panner(Coordinate=LinearInterpolate(A=@n28, B=@n29, Alpha=@n30.R)) {speedX=-0.18, speedY=0.25}) {samplerType=SAMPLERTYPE_LinearColor, texture=/Game/Global/Textures/Default/Flowmaps/T_SwirlFlow_02}
  n32 = Multiply(A=ComponentMask(None=@n31.RGB) {g=true, r=true}, B=ScalarParameter() {defaultValue=0.4, parameterName=Warp Strength})
  n33 = OneMinus(None=ComponentMask(None=TextureCoordinate()) {g=true})
  n34 = Panner(Coordinate=LinearInterpolate(A=@n27, B=Add(A=@n27, B=@n32), Alpha=Multiply(A=@n33, B=@n33))) {bFractionalPart=true, speedY=0.125}
  n35 = TextureCoordinate()
  n36 = ComponentMask(None=@n35) {r=true}
  n37 = Multiply(A=Multiply(A=@n36, B=@n36), B=OneMinus(None=@n36))
  n38 = ComponentMask(None=@n35) {g=true}
  n39 = Multiply(A=@n38, B=@n38)
  n40 = OneMinus(None=@n38)
  n41 = Multiply(
   A=Add(A=Saturate(None=@n26), B=TextureSample(UVs=@n34) {texture=&T_Inky_Smoke_Tile}.R)
   B=StaticSwitchParameter(True=Multiply(A=@n37, B=@n39) {constA=1}, False=Multiply(A=@n39, B=@n40)) {defaultValue=true, parameterName=Use Edge Fade}
  )
  n42 = NamedRerouteUsage()
  n43 = DepthFade(
   Opacity=Multiply(A=ScalarParameter() {defaultValue=10, parameterName=Alpha}, B=Multiply(A=@n41, B=@n42))
   FadeDistance=ScalarParameter() {defaultValue=5, parameterName=Fade Distance}
  )
  n44 = Multiply(
   A=MaterialFunctionCall(Distance=ScalarParameter() {defaultValue=150, parameterName=Near Camera Fade Distance}) {materialFunction=&Near_Camera_Fade}.Result
   B=@n43
  )
  n45 = OneMinus(None=VertexColor().G)
  n46 = ScalarParameter() {defaultValue=1, parameterName=s}
  n47 = Multiply(A=LinearInterpolate(B=@n5.R, Alpha=NamedRerouteUsage())) {constB=2}
  n48 = Multiply(A=@n43, B=Subtract(A=Power(Base=@n45, Exp=@n46), B=Saturate(None=@n47)))
  n49 = MaterialFunctionCall(Float2=TextureCoordinate()) {materialFunction=&BreakOutFloat2Components}
  Multiply(
   A=Multiply(A=StaticSwitchParameter(True=@n44, False=@n48) {parameterName=Near Camera Fade}, B=OneMinus(None=@n49.G))
   B=ScalarParameter() {defaultValue=1, parameterName=SwingFade}
  )
  n50 = ScalarParameter() {defaultValue=1, parameterName=FresEdge Exp}
  n51 = Fresnel(ExponentIn=@n50) {baseReflectFraction=0}
  n52 = OneMinus(None=VertexNormalWS())
  n53 = Power(Base=LinearInterpolate(A=OneMinus(None=@n51), B=Fresnel(ExponentIn=@n50, Normal=@n52) {baseReflectFraction=0}))
  NamedRerouteDeclaration(None=Saturate(None=Multiply(A=@n53) {constB=6})) {name=FresnelFadeEdges, nodeColor=<rgba 1 0 0.6797 1>}
  NamedRerouteDeclaration(None=@n31.G) {name=SmokyTexture, nodeColor=<rgba 1 0.4453 0 1>}
 ]
}
